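# Mono looping MetaSound for first-person crouch sliding, layering a surface loop with falling wind.
# Play or AssetSwitch picks a random BaseWaves entry (dirt by default) and feeds it to a crossfading surface-loop swapper patch.
# ZVelocity thresholds start and stop the wind loop; VelocityLength and ZVelocity set the respective layer gains before VolumeMultiplier scales the mix.
# The Mobile page keeps the same graph with one default wave.
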MetaSoundSource /FNGameplayCues/Athena/Sounds/Character/Sliding/MSS_CrouchSliding_Loop_1p
inputs [
 AssetSwitch {type=Trigger}
 BaseWaves {type=WaveAsset:Array}
 UE.Source.OnPlay {type=Trigger}
 VelocityLength {type=Float}
 VolumeMultiplier {type=Float}
 ZVelocity {type=Float}
]
interfaces [UE.OutputFormat.Mono:1.0, UE.Source:1.1]
outputs ["UE.OutputFormat.Mono.Audio:0" {type=Audio}]
pages [
 Default {
  is_build_graph=true
  input_defaults {
   AssetSwitch=null
   UE.Source.OnPlay=null
   VelocityLength=0
   VolumeMultiplier=1
   ZVelocity=0
   BaseWaves [
    /Game/Athena/Sounds/Character/Sliding/Waves/Dirt/Sliding_Dirt_Layer01_01
    /Game/Athena/Sounds/Character/Sliding/Waves/Dirt/Sliding_Dirt_Layer01_02
    /Game/Athena/Sounds/Character/Sliding/Waves/Dirt/Sliding_Dirt_Layer01_03
   ]
  }
  nodes [
   TriggerAnyTriggerAny2_1 = "TriggerAny.Trigger Any (2)"(In 1=^AssetSwitch, In 0=^UE.Source.OnPlay)
   ArrayRandomGet_0 = Array.Random Get:WaveAsset:Array(In Array=^BaseWaves, Next=@TriggerAnyTriggerAny2_1.Out) {also {defaults {Weights []}}}
   n1 = UE.InterpTo:Audio(Target=^ZVelocity) {also {defaults {Interp Time=0.25}}}
   TriggerAnyTriggerAny2_0 = "TriggerAny.Trigger Any (2)"(
    In 0=TriggerCompare.Clamp:Float(Compare=^UE.Source.OnPlay, A=^ZVelocity) {also {defaults {B=0.5, Type=3}}}.True
    In 1=UE.TriggerOnThreshold:Float(In=@n1.Value) {also {defaults {Threshold=0.5}}}.Out
   )
   WavePlayer_0 = UE.Wave Player:Mono(Play=@TriggerAnyTriggerAny2_0.Out, Stop=UE.TriggerOnThreshold:Float(In=@n1.Value) {also {defaults {Threshold=0.5, Type=1}}}.Out) {also {defaults {Loop="<MetaSoundBuilderResult.SUCCEEDED: 0>", Wave Asset=/Game/Athena/Sounds/Wind_Falling/AMB_Wind_Skydive_Vertical_Loop_01}}}
   AudioMixerAudioMixerMono2_0 = "AudioMixer.Audio Mixer (Mono, 2)"(
    In 0=/FNGameplayCues/Sounds/Fort_Audio_Blueprints/SurfaceTypeSound/MSP_SurfaceLoopSwapper(PlayNextSound=@ArrayRandomGet_0.On Next, NextSoundToPlay=@ArrayRandomGet_0.Value) {also {defaults {CrossfadeTime=0.2}}}.Out
    Gain 1=UE.Multiply:Float(PrimaryOperand=MapRange.MapRange:Float(In=^ZVelocity) {also {defaults {In Range A=0.5, Out Range B=1.4}}}.Out Value) {also {defaults {AdditionalOperands=0.6}}}.Out
    Gain 0=UE.Multiply:Float(PrimaryOperand=MapRange.MapRange:Float(In=^VelocityLength) {also {defaults {In Range A=0.5, Out Range A=0.5}}}.Out Value) {also {defaults {AdditionalOperands=1.8}}}.Out
    In 1=@WavePlayer_0.Out Mono
   )
   Multiply_0 = UE.Multiply:Audio by Float(AdditionalOperands=^VolumeMultiplier, PrimaryOperand=@AudioMixerAudioMixerMono2_0.Out)
  ]
  links [{dst="@UE.OutputFormat.Mono.Audio:0", src=Multiply_0.Out}]
 }
 Mobile {
  input_defaults {ZVelocity=0, BaseWaves [&Sliding_Dirt_Layer01_01]}
  nodes [
   TriggerAnyTriggerAny2_1 = "TriggerAny.Trigger Any (2)"(In 1=^AssetSwitch, In 0=^UE.Source.OnPlay)
   ArrayRandomGet_0 = Array.Random Get:WaveAsset:Array(In Array=^BaseWaves, Next=@TriggerAnyTriggerAny2_1.Out) {also {defaults {Weights []}}}
   n1 = UE.InterpTo:Audio(Target=^ZVelocity) {also {defaults {Interp Time=0.25}}}
   TriggerAnyTriggerAny2_0 = "TriggerAny.Trigger Any (2)"(
    In 0=TriggerCompare.Clamp:Float(Compare=^UE.Source.OnPlay, A=^ZVelocity) {also {defaults {B=0.5, Type=3}}}.True
    In 1=UE.TriggerOnThreshold:Float(In=@n1.Value) {also {defaults {Threshold=0.5}}}.Out
   )
   WavePlayer_0 = UE.Wave Player:Mono(Play=@TriggerAnyTriggerAny2_0.Out, Stop=UE.TriggerOnThreshold:Float(In=@n1.Value) {also {defaults {Threshold=0.5, Type=1}}}.Out) {also {defaults {Loop="<MetaSoundBuilderResult.SUCCEEDED: 0>", Wave Asset=&AMB_Wind_Skydive_Vertical_Loop_01}}}
   AudioMixerAudioMixerMono2_0 = "AudioMixer.Audio Mixer (Mono, 2)"(
    In 0=&MSP_SurfaceLoopSwapper(PlayNextSound=@ArrayRandomGet_0.On Next, NextSoundToPlay=@ArrayRandomGet_0.Value) {also {defaults {CrossfadeTime=0.2}}}.Out
    Gain 1=UE.Multiply:Float(PrimaryOperand=MapRange.MapRange:Float(In=^ZVelocity) {also {defaults {In Range A=0.5, Out Range B=1.4}}}.Out Value) {also {defaults {AdditionalOperands=0.6}}}.Out
    Gain 0=UE.Multiply:Float(PrimaryOperand=MapRange.MapRange:Float(In=^VelocityLength) {also {defaults {In Range A=0.5, Out Range A=0.5}}}.Out Value) {also {defaults {AdditionalOperands=1.8}}}.Out
    In 1=@WavePlayer_0.Out Mono
   )
   Multiply_0 = UE.Multiply:Audio by Float(AdditionalOperands=^VolumeMultiplier, PrimaryOperand=@AudioMixerAudioMixerMono2_0.Out)
  ]
  links [{dst="@UE.OutputFormat.Mono.Audio:0", src=Multiply_0.Out}]
 }
]
properties {
 assetImportData=&self:AssetImportData
 attenuationSettings=/Game/Sounds/Attenuation/Fort_Player_Footsteps_1P_Sprint_Athena_Att
 bLooping=true
 duration=10000
 qualitySetting=Fort_Weapons
 sampleRate=48000
 soundClassObject=&Fort_Player_Foley_1P_Athena
 volume=2.5
}
