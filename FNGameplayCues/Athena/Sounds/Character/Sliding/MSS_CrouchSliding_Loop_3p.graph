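# Looping mono MetaSound source for third-person crouch-slide audio, defaulting to dirt slide layers.
# On play or AssetSwitch, a random BaseWaves entry feeds MSP_SurfaceLoopSwapper, which crossfades to the new loop.
# Gain combines remapped VelocityLength and VolumeMultiplier, plus an Affiliation-selected level on the Default page.
# The Mobile page omits the affiliation scaling and defaults to a single wave.
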MetaSoundSource /FNGameplayCues/Athena/Sounds/Character/Sliding/MSS_CrouchSliding_Loop_3p
inputs [
 Affiliation {type=Int32}
 AssetSwitch {type=Trigger}
 BaseWaves {type=WaveAsset:Array}
 UE.Source.OnPlay {type=Trigger}
 VelocityLength {type=Float}
 VolumeMultiplier {type=Float}
]
interfaces [UE.OutputFormat.Mono:1.0, UE.Source:1.1]
outputs ["UE.OutputFormat.Mono.Audio:0" {type=Audio}]
pages [
 Default {
  is_build_graph=true
  input_defaults {
   Affiliation=null
   AssetSwitch=null
   UE.Source.OnPlay=null
   VelocityLength=0
   VolumeMultiplier=1
   BaseWaves [
    /Game/Athena/Sounds/Character/Sliding/Waves/Dirt/Sliding_Dirt_Layer01_01
    /Game/Athena/Sounds/Character/Sliding/Waves/Dirt/Sliding_Dirt_Layer01_02
    /Game/Athena/Sounds/Character/Sliding/Waves/Dirt/Sliding_Dirt_Layer01_03
   ]
  }
  nodes [
   TriggerAnyTriggerAny2_0 = "TriggerAny.Trigger Any (2)"(In 1=^AssetSwitch, In 0=^UE.Source.OnPlay)
   ArrayRandomGet_0 = Array.Random Get:WaveAsset:Array(In Array=^BaseWaves, Next=@TriggerAnyTriggerAny2_0.Out) {also {defaults {Weights []}}}
   CrossfadeTriggerRouteFloat4_0 = "Crossfade.Trigger Route (Float, 4):Float"(Crossfade Value=Convert.Float:Int32(Int32=^Affiliation).Float) {also {defaults {In 0=1, In 1=0.075, In 2=0.5, In 3=0.075}}}
   n1 = UE.Multiply:Float(PrimaryOperand=MapRange.MapRange:Float(In=^VelocityLength) {also {defaults {In Range A=0.5, Out Range A=0.5}}}.Out Value) {also {defaults {AdditionalOperands=1.8}}}
   Multiply_0 = UE.Multiply:Audio by Float(
    PrimaryOperand=/FNGameplayCues/Sounds/Fort_Audio_Blueprints/SurfaceTypeSound/MSP_SurfaceLoopSwapper(PlayNextSound=@ArrayRandomGet_0.On Next, NextSoundToPlay=@ArrayRandomGet_0.Value) {also {defaults {CrossfadeTime=0.2}}}.Out
    AdditionalOperands=UE.Multiply:Float(
     AdditionalOperands=@CrossfadeTriggerRouteFloat4_0.Out
     PrimaryOperand=UE.Multiply:Float(AdditionalOperands=^VolumeMultiplier, PrimaryOperand=@n1.Out).Out
    ).Out
   )
  ]
  links [{dst="@UE.OutputFormat.Mono.Audio:0", src=Multiply_0.Out}]
 }
 Mobile {
  input_defaults {BaseWaves [&Sliding_Dirt_Layer01_01]}
  nodes [
   TriggerAnyTriggerAny2_0 = "TriggerAny.Trigger Any (2)"(In 1=^AssetSwitch, In 0=^UE.Source.OnPlay)
   ArrayRandomGet_0 = Array.Random Get:WaveAsset:Array(In Array=^BaseWaves, Next=@TriggerAnyTriggerAny2_0.Out) {also {defaults {Weights []}}}
   n1 = MapRange.MapRange:Float(In=^VelocityLength) {also {defaults {In Range A=0.5, Out Range A=0.5}}}
   Multiply_0 = UE.Multiply:Audio by Float(
    PrimaryOperand=&MSP_SurfaceLoopSwapper(PlayNextSound=@ArrayRandomGet_0.On Next, NextSoundToPlay=@ArrayRandomGet_0.Value) {also {defaults {CrossfadeTime=0.2}}}.Out
    AdditionalOperands=UE.Multiply:Float(
     AdditionalOperands=^VolumeMultiplier
     PrimaryOperand=UE.Multiply:Float(PrimaryOperand=@n1.Out Value) {also {defaults {AdditionalOperands=1.8}}}.Out
    ).Out
   )
  ]
  links [{dst="@UE.OutputFormat.Mono.Audio:0", src=Multiply_0.Out}]
 }
]
properties {
 assetImportData=&self:AssetImportData
 attenuationSettings=/Game/Sounds/Attenuation/Fort_Player_Footsteps_3P_Sprint_Athena_Att
 bLooping=true
 duration=10000
 maxDistance=3584
 sampleRate=48000
 soundClassObject=&Fort_Player_Foley_3P_Enemy_Athena
 volume=2
}
