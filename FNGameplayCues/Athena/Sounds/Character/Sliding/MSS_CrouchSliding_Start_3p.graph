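# MetaSound Source preset for the third-person crouch-slide start one-shot, built on the first-person version MSS_CrouchSliding_Start_1p.
# It inherits the parent's graph and supplies dirt slide and heavy water splash wave arrays, alongside affiliation, volume multiplier and water immersion inputs.
# Output is mono, with third-person sprint footstep attenuation, the enemy 3P foley sound class and a raised source volume.
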MetaSoundSource /FNGameplayCues/Athena/Sounds/Character/Sliding/MSS_CrouchSliding_Start_3p
preset_parent=&/MSS_CrouchSliding_Start_1p
inputs [
 Affiliation {type=Int32}
 UE.Source.OnPlay {type=Trigger}
 VolumeMultiplier {type=Float}
 WaterImmersion {type=Float}
 WaterWaves {type=WaveAsset:Array}
 Waves {type=WaveAsset:Array}
]
interfaces [UE.OutputFormat.Mono:1.0, UE.Source.OneShot:1.0, UE.Source:1.1]
outputs [
 "UE.OutputFormat.Mono.Audio:0" {type=Audio}
 UE.Source.OneShot.OnFinished {type=Trigger}
]
pages [Default {input_defaults {
   Affiliation=0
   UE.Source.OnPlay=null
   VolumeMultiplier=1
   WaterImmersion=0
   WaterWaves [
    /FNGameplayCues/Sounds/Physics/Objects/WaterSplash/MediumObject/Phys_Medium_Object_Splash_Heavy_04
    /FNGameplayCues/Sounds/Physics/Objects/WaterSplash/MediumObject/Phys_Medium_Object_Splash_Heavy_03
    /FNGameplayCues/Sounds/Physics/Objects/WaterSplash/MediumObject/Phys_Medium_Object_Splash_Heavy_02
    /FNGameplayCues/Sounds/Physics/Objects/WaterSplash/MediumObject/Phys_Medium_Object_Splash_Heavy_01
   ]
   Waves [
    /FNGameplayCues/Athena/Sounds/Character/Sliding/Waves/Dirt/Sliding_Dirt_OneShot_06
    /FNGameplayCues/Athena/Sounds/Character/Sliding/Waves/Dirt/Sliding_Dirt_OneShot_05
    /FNGameplayCues/Athena/Sounds/Character/Sliding/Waves/Dirt/Sliding_Dirt_OneShot_04
    /FNGameplayCues/Athena/Sounds/Character/Sliding/Waves/Dirt/Sliding_Dirt_OneShot_03
    /FNGameplayCues/Athena/Sounds/Character/Sliding/Waves/Dirt/Sliding_Dirt_OneShot_02
    /FNGameplayCues/Athena/Sounds/Character/Sliding/Waves/Dirt/Sliding_Dirt_OneShot_01
   ]
  }}]
properties {
 assetImportData=&self:AssetImportData
 attenuationSettings=/Game/Sounds/Attenuation/Fort_Player_Footsteps_3P_Sprint_Athena_Att
 maxDistance=3584
 sampleRate=48000
 soundClassObject=&Fort_Player_Foley_3P_Enemy_Athena
 volume=1.5
}
