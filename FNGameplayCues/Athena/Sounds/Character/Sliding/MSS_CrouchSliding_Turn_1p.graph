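# One-shot mono MetaSound for first-person crouch-slide turn foley, defaulting to dirt slide-stop waves.
# On play, a slight random float is generated, then a random wave is picked from the Waves input and played.
# That float passes through an affiliation mix patch weighting enemy louder than non-enemy, then is multiplied by VolumeMultiplier to set mixer gain.
# The random variation affects volume only, and wave completion fires OnFinished.
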MetaSoundSource /FNGameplayCues/Athena/Sounds/Character/Sliding/MSS_CrouchSliding_Turn_1p
inputs [
 Affiliation {type=Int32}
 UE.Source.OnPlay {type=Trigger}
 VolumeMultiplier {type=Float}
 Waves {type=WaveAsset:Array}
]
interfaces [UE.OutputFormat.Mono:1.0, UE.Source.OneShot:1.0, UE.Source:1.1]
outputs [
 "UE.OutputFormat.Mono.Audio:0" {type=Audio}
 UE.Source.OneShot.OnFinished {type=Trigger}
]
pages [Default {
  input_defaults {
   Affiliation=0
   UE.Source.OnPlay=null
   VolumeMultiplier=1
   Waves [
    /Game/Athena/Sounds/Character/Sliding/Waves/Dirt/Sliding_Dirt_Stop_06
    /Game/Athena/Sounds/Character/Sliding/Waves/Dirt/Sliding_Dirt_Stop_05
    /Game/Athena/Sounds/Character/Sliding/Waves/Dirt/Sliding_Dirt_Stop_04
    /Game/Athena/Sounds/Character/Sliding/Waves/Dirt/Sliding_Dirt_Stop_03
    /Game/Athena/Sounds/Character/Sliding/Waves/Dirt/Sliding_Dirt_Stop_02
    /Game/Athena/Sounds/Character/Sliding/Waves/Dirt/Sliding_Dirt_Stop_01
   ]
  }
  nodes [
   n1 = Convert.Enum:FortAudioAffiliation:Int32(Int32=^Affiliation)
   n2 = UE.RandomFloat(Next=^UE.Source.OnPlay) {also {defaults {Max=1.05, Min=0.95}}}
   n3 = UE.Reroute(Value=UE.Reroute(Value=@n2.Value).Value)
   ArrayRandomGet_0 = Array.Random Get:WaveAsset:Array(In Array=^Waves, Next=@n2.On Next) {also {defaults {Weights []}}}
   WavePlayer_0 = UE.Wave Player:Mono(Play=@ArrayRandomGet_0.On Next, Wave Asset=@ArrayRandomGet_0.Value)
   AudioMixerAudioMixerMono2_0 = "AudioMixer.Audio Mixer (Mono, 2)"(
    Gain 0=UE.Multiply:Float(
     AdditionalOperands=^VolumeMultiplier
     PrimaryOperand=&MSP_AffiliationMixFloat(Affiliation=@n1."Enum:FortAudioAffiliation", In Value=@n3.Value) {also {defaults {Enemy=0.7, Non-Enemy=0.175}}}.Out
    ).Out
    In 0=@WavePlayer_0.Out Mono
   )
  ]
  links [
   {dst="@UE.OutputFormat.Mono.Audio:0", src=AudioMixerAudioMixerMono2_0.Out}
   {dst="@UE.Source.OneShot.OnFinished", src=WavePlayer_0.On Finished}
  ]
 }]
properties {
 assetImportData=&self:AssetImportData
 attenuationSettings=/Game/Sounds/Attenuation/Fort_Player_Footsteps_1P_Sprint_Athena_Att
 sampleRate=48000
 soundClassObject=&Fort_Player_Foley_1P_Athena
 volume=0.7
}
