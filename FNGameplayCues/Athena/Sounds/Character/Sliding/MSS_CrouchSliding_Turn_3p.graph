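# Third-person MetaSound for the turn sound during crouch sliding, heard from other players' characters.
# It is a preset of the first-person turn MetaSound, inheriting its graph and supplying dirt slide stop wave variations plus affiliation and volume multiplier inputs.
# It outputs mono one-shot audio with third-person sprint footstep attenuation, an enemy third-person foley sound class and a reduced base volume.
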MetaSoundSource /FNGameplayCues/Athena/Sounds/Character/Sliding/MSS_CrouchSliding_Turn_3p
preset_parent=&/MSS_CrouchSliding_Turn_1p
inputs [
 Affiliation {type=Int32}
 UE.Source.OnPlay {type=Trigger}
 VolumeMultiplier {type=Float}
 Waves {type=WaveAsset:Array}
]
interfaces [UE.OutputFormat.Mono:1.0, UE.Source.OneShot:1.0, UE.Source:1.1]
outputs [
 "UE.OutputFormat.Mono.Audio:0" {type=Audio}
 UE.Source.OneShot.OnFinished {type=Trigger}
]
pages [Default {input_defaults {
   Affiliation=0
   UE.Source.OnPlay=null
   VolumeMultiplier=1
   Waves [
    /Game/Athena/Sounds/Character/Sliding/Waves/Dirt/Sliding_Dirt_Stop_06
    /Game/Athena/Sounds/Character/Sliding/Waves/Dirt/Sliding_Dirt_Stop_05
    /Game/Athena/Sounds/Character/Sliding/Waves/Dirt/Sliding_Dirt_Stop_04
    /Game/Athena/Sounds/Character/Sliding/Waves/Dirt/Sliding_Dirt_Stop_03
    /Game/Athena/Sounds/Character/Sliding/Waves/Dirt/Sliding_Dirt_Stop_02
    /Game/Athena/Sounds/Character/Sliding/Waves/Dirt/Sliding_Dirt_Stop_01
   ]
  }}]
properties {
 assetImportData=&self:AssetImportData
 attenuationSettings=/Game/Sounds/Attenuation/Fort_Player_Footsteps_3P_Sprint_Athena_Att
 maxDistance=3584
 soundClassObject=&Fort_Player_Foley_3P_Enemy_Athena
 volume=0.7
}
